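# Translucent lit Niagara sprite material for dust effects in a test event folder, using an 8x8 falling-mist flipbook with volumetric directional lighting.
# SubUV red is tinted by particle color, while blue, raised to a dynamic Alpha Contrast power and scaled by particle alpha, forms opacity.
# Depth fade and near-camera fade soften that opacity.
# An emissive-adjustment function takes particle color, and a sphere normal texture tiles to match the flipbook grid.
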Material /BRCosmetics/Athena/Events/Test/M_Dust_8x8_02
graph {
 properties {
  bUsedWithNiagaraSprites=true
  blendMode=BLEND_Translucent
  pixelDepthOffsetMode=PDOM_Legacy
  translucencyDirectionalLightingIntensity=0.9
  translucencyLightingMode=TLM_VolumetricDirectional
  usedShadingModels=Default Lit
 }
 nodes [
  MaterialFunctionCall(Emissive=ParticleColor().RGB) {materialFunction=&MF_FX_EmissiveAdjustments}
  n1 = ParticleSubUV!(UVs, Tex, Apply View MipBias) {samplerType=SAMPLERTYPE_LinearColor, texture=/Game/Effects/Fort_Effects/Textures/Smoke/T_FallingMist_8x8_01}
  n2 = Power(
   Base=@n1.B
   Exp=DynamicParameter() {defaultValue=<rgba 1 0.1 1 1>, paramNames [Alpha Contrast, Emissive Boost, Param3, Param4]}.Alpha Contrast
  )
  n3 = ParticleColor()
  Multiply(
   A=DepthFade(Opacity=Multiply(A=@n2, B=@n3.A), FadeDistance=ScalarParameter() {defaultValue=350, parameterName=Depth Fade Distance})
   B=MaterialFunctionCall(Distance=ScalarParameter() {defaultValue=500, parameterName=NearCameraFadeDist}) {materialFunction=&Near_Camera_Fade}.Result
  )
  Multiply(A=@n1.R, B=@n3.RGB)
  Multiply()
  TextureSample(UVs=TextureCoordinate() {uTiling=8, vTiling=8}) {samplerType=SAMPLERTYPE_Normal, texture=/Game/Environments/World/Backgrounds/Universal/TechContent/Textures/T_Spherenormal_01}
 ]
}
